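# Animation Blueprint for the F_MED_FuzzyGlam face accessory, adding secondary motion to hair, feathers and an earring.
# Crouch state blends between a Control Rig/AnimDynamics branch with corrective bone offsets and a head-space RigidBody simulation.
# Further AnimDynamics and offsets follow, then constraints weight hair toward neck, spine and clavicles when backbling is detected.
# The EventGraph updates gravity settings from a parameter asset, head-up alphas for targeting, and backbling flags.
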AnimBlueprint /Hermes/Characters/Player/Female/Medium/Bodies/F_MED_FuzzyGlam/Meshes/Parts/F_MED_FuzzyGlam_FaceAcc_AnimBP
target_skeleton=&/F_MED_FuzzyGlam_FaceAcc_Skeleton
graphs [
 AnimGraph {nodes [
   n1 = "Transform (Modify) Bone - Bone: dyn_r_hair_b_base"(Alpha=1, Translation="-3.129,-6.131,-1.768", Rotation="25.622,176.807,-141.118", Scale="1.000,0.813,1.000") {
    rotationMode=BMM_Additive
    rotationSpace=BCS_BoneSpace
    scaleMode=BMM_Additive
    scaleSpace=BCS_BoneSpace
    translationMode=BMM_Additive
    translationSpace=BCS_BoneSpace
    alphaScaleBiasClamp {bInterpResult=true}
    boneToModify {boneName=dyn_r_hair_b_base}
    rotation=<rot 25.622 176.807 -141.118>
    scale {y=0.8132}
    translation=<-3.129 -6.131 -1.768>
   }
   "Transform (Modify) Bone - Bone: dyn_r_hair_b_02"(ComponentPose=@n1, Alpha=1, Translation="6.340,4.673,-6.488", Rotation="-19.865,15.398,-23.490", Scale="0.446,1.000,1.000") {
    rotationMode=BMM_Additive
    rotationSpace=BCS_BoneSpace
    scaleMode=BMM_Additive
    scaleSpace=BCS_BoneSpace
    translationMode=BMM_Additive
    translationSpace=BCS_BoneSpace
    alphaScaleBiasClamp {bInterpResult=true}
    boneToModify {boneName=dyn_r_hair_b_02}
    rotation=<rot -19.865 15.398 -23.49>
    scale {x=0.446}
    translation=<6.34 4.673 -6.488>
   }
   "Transform (Modify) Bone - Bone: dyn_r_hair_b_02"(Alpha=1, Translation="0,0,0", Rotation="-5.799,0.000,-0.001", Scale="1,1,1") {
    rotationMode=BMM_Additive
    rotationSpace=BCS_BoneSpace
    scaleMode=BMM_Additive
    scaleSpace=BCS_BoneSpace
    translationMode=BMM_Additive
    translationSpace=BCS_BoneSpace
    alphaScaleBiasClamp {bInterpResult=true, interpSpeedDecreasing=1, interpSpeedIncreasing=1}
    boneToModify {boneName=dyn_r_hair_b_02}
    rotation=<rot -5.799 0 -0.001>
   }
   n2 = "Transform (Modify) Bone - Bone: dyn_l_hair_b_base"(Alpha=1, Translation="2.323,-0.692,-0.712", Rotation="-34.926,-27.279,18.745", Scale="1,1,1") {
    rotationMode=BMM_Additive
    rotationSpace=BCS_BoneSpace
    translationMode=BMM_Additive
    translationSpace=BCS_BoneSpace
    alphaScaleBiasClamp {bInterpResult=true}
    boneToModify {boneName=dyn_l_hair_b_base}
    rotation=<rot -34.926 -27.279 18.745>
    translation=<2.323 -0.692 -0.712>
   }
   n3 = "Transform (Modify) Bone - Bone: dyn_m_hair_base"(ComponentPose=@n2, Alpha=1, Translation="0,0,0", Rotation="9.980,-17.323,-1.779", Scale="1,1,1") {
    rotationMode=BMM_Additive
    rotationSpace=BCS_BoneSpace
    scaleMode=BMM_Additive
    scaleSpace=BCS_BoneSpace
    translationMode=BMM_Additive
    translationSpace=BCS_BoneSpace
    alphaScaleBiasClamp {bInterpResult=true}
    boneToModify {boneName=dyn_m_hair_base}
    rotation=<rot 9.98 -17.323 -1.779>
   }
   n4 = "Transform (Modify) Bone - Bone: dyn_r_hair_b_base"(ComponentPose=@n3, Alpha=1, Translation="0,0,0", Rotation="25.622,176.807,-141.118", Scale="1.000,0.813,1.000") {
    rotationMode=BMM_Additive
    rotationSpace=BCS_BoneSpace
    scaleMode=BMM_Additive
    scaleSpace=BCS_BoneSpace
    translationMode=BMM_Additive
    translationSpace=BCS_BoneSpace
    alphaScaleBiasClamp {bInterpResult=true}
    boneToModify {boneName=dyn_r_hair_b_base}
    rotation=<rot 25.622 176.807 -141.118>
    scale {y=0.8132}
   }
   n5 = "Transform (Modify) Bone - Bone: dyn_r_hair_b_02"(ComponentPose=@n4, Alpha=1, Translation="0,0,0", Rotation="-19.865,15.398,-23.490", Scale="0.446,1.000,1.000") {
    rotationMode=BMM_Additive
    rotationSpace=BCS_BoneSpace
    scaleMode=BMM_Additive
    scaleSpace=BCS_BoneSpace
    translationMode=BMM_Additive
    translationSpace=BCS_BoneSpace
    alphaScaleBiasClamp {bInterpResult=true}
    boneToModify {boneName=dyn_r_hair_b_02}
    rotation=<rot -19.865 15.398 -23.49>
    scale {x=0.446}
   }
   n6 = "Transform (Modify) Bone - Bone: dyn_l_hair_a_base"(ComponentPose=@n5, Alpha=1, Translation="0,0,0", Rotation="0.006,-70.000,-0.003", Scale="1,1,1") {
    rotationMode=BMM_Additive
    rotationSpace=BCS_BoneSpace
    scaleMode=BMM_Additive
    scaleSpace=BCS_BoneSpace
    translationMode=BMM_Additive
    translationSpace=BCS_BoneSpace
    alphaScaleBiasClamp {bInterpResult=true}
    boneToModify {boneName=dyn_l_hair_a_base}
    rotation=<rot 0.006 -69.9999 -0.003>
   }
   n7 = "Transform (Modify) Bone - Bone: dyn_l_hair_a_02"(ComponentPose=@n6, Alpha=1, Translation="0.271,2.332,0.016", Rotation="-0.020,90.000,-0.015", Scale="1,1,1") {
    rotationMode=BMM_Additive
    rotationSpace=BCS_BoneSpace
    scaleMode=BMM_Additive
    scaleSpace=BCS_BoneSpace
    translationMode=BMM_Additive
    translationSpace=BCS_BoneSpace
    alphaScaleBiasClamp {bInterpResult=true}
    boneToModify {boneName=dyn_l_hair_a_02}
    rotation=<rot -0.02 90 -0.015>
    translation=<0.271 2.332 0.016>
   }
   "Transform (Modify) Bone - Bone: dyn_r_hair_a_base1"(ComponentPose=@n7, Alpha=1, Translation="0,0,0", Rotation="0,35,0", Scale="1,1,1") {
    rotationMode=BMM_Additive
    rotationSpace=BCS_BoneSpace
    scaleMode=BMM_Additive
    scaleSpace=BCS_BoneSpace
    translationMode=BMM_Additive
    translationSpace=BCS_BoneSpace
    alphaScaleBiasClamp {bInterpResult=true}
    boneToModify {boneName=dyn_r_hair_a_base1}
    rotation {yaw=35}
   }
   n8 = "Transform (Modify) Bone - Bone: dyn_l_hair_b_base"(Alpha=1, Translation="4.461,-1.265,0.368", Rotation="20.104,-23.447,-11.827", Scale="1,1,1") {
    rotationMode=BMM_Additive
    rotationSpace=BCS_BoneSpace
    translationMode=BMM_Additive
    translationSpace=BCS_BoneSpace
    alphaScaleBiasClamp {bInterpResult=true, interpSpeedDecreasing=1, interpSpeedIncreasing=1}
    boneToModify {boneName=dyn_l_hair_b_base}
    rotation=<rot 20.104 -23.447 -11.827>
    translation=<4.461 -1.265 0.368>
   }
   n9 = "Transform (Modify) Bone - Bone: dyn_m_hair_base"(ComponentPose=@n8, Alpha=1, Translation="0,0,0", Rotation="26.990,0.816,-3.187", Scale="1,1,1") {
    rotationMode=BMM_Additive
    rotationSpace=BCS_BoneSpace
    scaleMode=BMM_Additive
    scaleSpace=BCS_BoneSpace
    translationMode=BMM_Additive
    translationSpace=BCS_BoneSpace
    alphaScaleBiasClamp {bInterpResult=true, interpSpeedDecreasing=1, interpSpeedIncreasing=1}
    boneToModify {boneName=dyn_m_hair_base}
    rotation=<rot 26.99 0.816 -3.187>
   }
   n10 = "Transform (Modify) Bone - Bone: dyn_r_hair_b_base"(ComponentPose=@n9, Alpha=1, Translation="-0.854,0.073,-0.285", Rotation="40.424,-49.854,-41.941", Scale="0.391,1.000,1.000") {
    rotationMode=BMM_Additive
    rotationSpace=BCS_BoneSpace
    scaleMode=BMM_Additive
    scaleSpace=BCS_BoneSpace
    translationMode=BMM_Additive
    translationSpace=BCS_BoneSpace
    alphaScaleBiasClamp {bInterpResult=true, interpSpeedDecreasing=1, interpSpeedIncreasing=1}
    boneToModify {boneName=dyn_r_hair_b_base}
    rotation=<rot 40.424 -49.854 -41.941>
    scale {x=0.391}
    translation=<-0.854 0.073 -0.285>
   }
   n11 = "Transform (Modify) Bone - Bone: dyn_r_hair_b_02"(ComponentPose=@n10, Alpha=1, Translation="0,0,0", Rotation="9.192,-11.055,-5.866", Scale="1,1,1") {
    rotationMode=BMM_Additive
    rotationSpace=BCS_BoneSpace
    scaleMode=BMM_Additive
    scaleSpace=BCS_BoneSpace
    translationMode=BMM_Additive
    translationSpace=BCS_BoneSpace
    alphaScaleBiasClamp {bInterpResult=true, interpSpeedDecreasing=1, interpSpeedIncreasing=1}
    boneToModify {boneName=dyn_r_hair_b_02}
    rotation=<rot 9.1919 -11.055 -5.866>
   }
   n12 = "Transform (Modify) Bone - Bone: dyn_r_hair_a_base1"(ComponentPose=@n11, Alpha=1, Translation="0,0,0", Rotation="0,0,0", Scale="1.000,1.000,-0.105") {
    rotationMode=BMM_Additive
    rotationSpace=BCS_BoneSpace
    scaleMode=BMM_Additive
    scaleSpace=BCS_BoneSpace
    translationMode=BMM_Additive
    translationSpace=BCS_BoneSpace
    alphaScaleBiasClamp {bInterpResult=true, interpSpeedDecreasing=1, interpSpeedIncreasing=1}
    boneToModify {boneName=dyn_r_hair_a_base1}
    scale {z=-0.105}
   }
   n13 = "Transform (Modify) Bone - Bone: dyn_l_hair_b_base"!(ComponentPose=@n12, bAlphaBoolEnabled:bool=True, Translation="0,0,0", Rotation="3.203,-6.360,7.062", Scale="1,1,1") -> (Pose) {
    alphaInputType=Bool
    rotationMode=BMM_Additive
    rotationSpace=BCS_BoneSpace
    translationMode=BMM_Additive
    translationSpace=BCS_BoneSpace
    alphaBoolBlend {blendInTime=0.3, blendOutTime=0.3}
    boneToModify {boneName=dyn_l_hair_b_base}
    rotation=<rot 3.203 -6.36 7.062>
   }
   n14 = "Transform (Modify) Bone - Bone: dyn_m_hair_base"!(ComponentPose=@n13, bAlphaBoolEnabled:bool=True, Translation="1.413,-0.003,-0.552", Rotation="5.391,10.138,-21.948", Scale="1,1,1") -> (Pose) {
    alphaInputType=Bool
    rotationMode=BMM_Additive
    rotationSpace=BCS_BoneSpace
    translationMode=BMM_Additive
    translationSpace=BCS_BoneSpace
    alphaBoolBlend {blendInTime=0.3, blendOutTime=0.3}
    boneToModify {boneName=dyn_m_hair_base}
    rotation=<rot 5.3909 10.138 -21.948>
    translation=<1.413 -0.003 -0.552>
   }
   "Transform (Modify) Bone - Bone: dyn_r_hair_b_base"!(ComponentPose=@n14, bAlphaBoolEnabled:bool=True, Translation="0,0,0", Rotation="14.384,14.534,-22.548", Scale="1,1,1") -> (Pose) {
    alphaInputType=Bool
    rotationMode=BMM_Additive
    rotationSpace=BCS_BoneSpace
    translationMode=BMM_Additive
    translationSpace=BCS_BoneSpace
    alphaBoolBlend {blendInTime=0.3, blendOutTime=0.3}
    boneToModify {boneName=dyn_r_hair_b_base}
    rotation=<rot 14.384 14.534 -22.548>
   }
   n15 = Control Rig(Source=Copy Pose From Mesh() {bUseAttachedParent=true}) {controlRigAssetReference {blueprintRigClass=&/F_MED_FuzzyGlam_FaceAcc_CtrlRig.F_MED_FuzzyGlam_FaceAcc_CtrlRig_C}}
   n16 = Anim Dynamics - Bone: dyn_l_hair_b_01!(ComponentPose=Local To Component(LocalPose=@n15), Alpha=1, GravityOverride:Vector="0,0,-980", ExternalForce="0,0,0") -> (Pose) {
    angularDampingOverride=1
    angularSpringConstant=10000
    bAngularSpring=true
    bOverrideAngularDamping=true
    bUseGravityOverride=true
    gravityScale=0.35
    simulationSpace=RootRelative
    boundBone {boneName=dyn_l_hair_b_01}
    componentAppliedLinearAccClamp=<5000 5000 5000>
    componentLinearAccScale=<1 1 1>
    componentLinearVelScale=<1 1 1>
    gravityOverride {z=-980}
    physicsBodyDefinitions [{
      boundBone {boneName=dyn_l_hair_b_01}
      boxExtents=<20 30 30>
      constraintSetup {angularLimitsMax=<0 5 5>, angularLimitsMin=<0 -30 -5>, angularTarget=<1 0 0>}
      localJointOffset=<3 0 0>
     }]
   }
   n17 = Anim Dynamics - Bone: dyn_l_hair_b_02!(ComponentPose=@n16, Alpha=1, GravityOverride:Vector="0,0,-980", ExternalForce="0,0,0") -> (Pose) {
    angularDampingOverride=1
    angularSpringConstant=10000
    bAngularSpring=true
    bOverrideAngularDamping=true
    bUseGravityOverride=true
    gravityScale=0.35
    simulationSpace=RootRelative
    boundBone {boneName=dyn_l_hair_b_02}
    componentAppliedLinearAccClamp=<5000 5000 5000>
    componentLinearAccScale=<1 1 1>
    componentLinearVelScale=<1 1 1>
    gravityOverride {z=-980}
    physicsBodyDefinitions [{
      boundBone {boneName=dyn_l_hair_b_02}
      boxExtents=<20 30 30>
      constraintSetup {angularLimitsMax=<0 5 5>, angularLimitsMin=<0 -30 -5>, angularTarget=<1 0 0>}
      localJointOffset=<3 0 0>
     }]
   }
   n18 = Anim Dynamics - Bone: dyn_m_hair_01!(ComponentPose=@n17, Alpha=1, GravityOverride:Vector="0,0,-980", ExternalForce="0,0,0") -> (Pose) {
    angularDampingOverride=1
    angularSpringConstant=10000
    bAngularSpring=true
    bOverrideAngularDamping=true
    bUseGravityOverride=true
    gravityScale=0.35
    simulationSpace=RootRelative
    boundBone {boneName=dyn_m_hair_01}
    componentAppliedLinearAccClamp=<5000 5000 5000>
    componentLinearAccScale=<1 1 1>
    componentLinearVelScale=<1 1 1>
    gravityOverride {z=-980}
    physicsBodyDefinitions [{
      boundBone {boneName=dyn_m_hair_01}
      boxExtents=<20 30 30>
      constraintSetup {angularLimitsMax=<0 5 5>, angularLimitsMin=<0 -30 -5>, angularTarget=<1 0 0>}
      localJointOffset=<3 0 0>
     }]
   }
   n19 = Anim Dynamics - Bone: dyn_m_hair_02!(ComponentPose=@n18, Alpha=1, GravityOverride:Vector="0,0,-980", ExternalForce="0,0,0") -> (Pose) {
    angularDampingOverride=1
    angularSpringConstant=10000
    bAngularSpring=true
    bOverrideAngularDamping=true
    bUseGravityOverride=true
    gravityScale=0.35
    simulationSpace=RootRelative
    boundBone {boneName=dyn_m_hair_02}
    componentAppliedLinearAccClamp=<5000 5000 5000>
    componentLinearAccScale=<1 1 1>
    componentLinearVelScale=<1 1 1>
    gravityOverride {z=-980}
    physicsBodyDefinitions [{
      boundBone {boneName=dyn_m_hair_02}
      boxExtents=<20 30 30>
      constraintSetup {angularLimitsMax=<0 5 5>, angularLimitsMin=<0 -30 -5>, angularTarget=<1 0 0>}
      localJointOffset=<3 0 0>
     }]
   }
   n20 = Anim Dynamics - Bone: dyn_r_hair_b_01!(ComponentPose=@n19, Alpha=1, GravityOverride:Vector="0,0,-980", ExternalForce="0,0,0") -> (Pose) {
    angularDampingOverride=1
    angularSpringConstant=10000
    bAngularSpring=true
    bOverrideAngularDamping=true
    bUseGravityOverride=true
    gravityScale=0.35
    simulationSpace=RootRelative
    boundBone {boneName=dyn_r_hair_b_01}
    componentAppliedLinearAccClamp=<5000 5000 5000>
    componentLinearAccScale=<1 1 1>
    componentLinearVelScale=<1 1 1>
    gravityOverride {z=-980}
    physicsBodyDefinitions [{
      boundBone {boneName=dyn_r_hair_b_01}
      boxExtents=<20 30 30>
      constraintSetup {angularLimitsMax=<0 5 30>, angularLimitsMin=<0 -5 -5>, angularTarget=<1 0 0>}
      localJointOffset=<3 0 0>
     }]
   }
   n21 = Anim Dynamics - Bone: dyn_r_hair_b_02!(ComponentPose=@n20, Alpha=1, GravityOverride:Vector="0,0,-980", ExternalForce="0,0,0") -> (Pose) {
    angularDampingOverride=1
    angularSpringConstant=10000
    bAngularSpring=true
    bOverrideAngularDamping=true
    bUseGravityOverride=true
    gravityScale=0.35
    simulationSpace=RootRelative
    boundBone {boneName=dyn_r_hair_b_02}
    componentAppliedLinearAccClamp=<5000 5000 5000>
    componentLinearAccScale=<1 1 1>
    componentLinearVelScale=<1 1 1>
    gravityOverride {z=-980}
    physicsBodyDefinitions [{
      boundBone {boneName=dyn_r_hair_b_02}
      boxExtents=<20 30 30>
      constraintSetup {angularLimitsMax=<0 5 30>, angularLimitsMin=<0 -5 -5>, angularTarget=<1 0 0>}
      localJointOffset=<3 0 0>
     }]
   }
   n22 = "Transform (Modify) Bone - Bone: dyn_l_hair_b_base"!(ComponentPose=@n21, Alpha:float=1, Translation="4.461,-1.265,0.368", Rotation="23.400,-16.811,-1.212", Scale="1,1,1") -> (Pose) {
    rotationMode=BMM_Additive
    rotationSpace=BCS_BoneSpace
    translationMode=BMM_Additive
    translationSpace=BCS_BoneSpace
    alphaScaleBiasClamp {bInterpResult=true, interpSpeedDecreasing=1, interpSpeedIncreasing=1}
    boneToModify {boneName=dyn_l_hair_b_base}
    rotation=<rot 23.4 -16.811 -1.212>
    translation=<4.461 -1.265 0.368>
   }
   n23 = "Transform (Modify) Bone - Bone: dyn_m_hair_base"!(ComponentPose=@n22, Alpha:float=1, Translation="0,0,0", Rotation="16.200,1.261,-3.097", Scale="1,1,1") -> (Pose) {
    rotationMode=BMM_Additive
    rotationSpace=BCS_BoneSpace
    scaleMode=BMM_Additive
    scaleSpace=BCS_BoneSpace
    translationMode=BMM_Additive
    translationSpace=BCS_BoneSpace
    alphaScaleBiasClamp {bInterpResult=true, interpSpeedDecreasing=1, interpSpeedIncreasing=1}
    boneToModify {boneName=dyn_m_hair_base}
    rotation=<rot 16.2 1.261 -3.097>
   }
   n24 = "Transform (Modify) Bone - Bone: dyn_r_hair_b_base"!(ComponentPose=@n23, Alpha:float=1, Translation="-1.355,-0.222,-0.874", Rotation="58.732,-30.560,6.617", Scale="0.391,1.000,1.000") -> (Pose) {
    rotationMode=BMM_Additive
    rotationSpace=BCS_BoneSpace
    scaleMode=BMM_Additive
    scaleSpace=BCS_BoneSpace
    translationMode=BMM_Additive
    translationSpace=BCS_BoneSpace
    alphaScaleBiasClamp {bInterpResult=true, interpSpeedDecreasing=1, interpSpeedIncreasing=1}
    boneToModify {boneName=dyn_r_hair_b_base}
    rotation=<rot 58.732 -30.56 6.617>
    scale {x=0.391}
    translation=<-1.355 -0.222 -0.874>
   }
   n25 = "Transform (Modify) Bone - Bone: dyn_r_hair_b_02"(ComponentPose=@n24, Alpha=1, Translation="7.351,2.424,-0.359", Rotation="-24.677,-41.856,-69.128", Scale="1,1,1") {
    rotationMode=BMM_Additive
    rotationSpace=BCS_BoneSpace
    scaleMode=BMM_Additive
    scaleSpace=BCS_BoneSpace
    translationMode=BMM_Additive
    translationSpace=BCS_BoneSpace
    alphaScaleBiasClamp {bInterpResult=true, interpSpeedDecreasing=1, interpSpeedIncreasing=1}
    boneToModify {boneName=dyn_r_hair_b_02}
    rotation=<rot -24.677 -41.856 -69.128>
    translation=<7.351 2.424 -0.359>
   }
   n26 = "Transform (Modify) Bone - Bone: dyn_r_hair_a_base1"(ComponentPose=@n25, Alpha=1, Translation="-2.720,-2.025,-0.538", Rotation="-10.643,45.682,-2.695", Scale="1.000,1.000,-0.105") {
    rotationMode=BMM_Additive
    rotationSpace=BCS_BoneSpace
    scaleMode=BMM_Additive
    scaleSpace=BCS_BoneSpace
    translationMode=BMM_Additive
    translationSpace=BCS_BoneSpace
    alphaScaleBiasClamp {bInterpResult=true, interpSpeedDecreasing=1, interpSpeedIncreasing=1}
    boneToModify {boneName=dyn_r_hair_a_base1}
    rotation=<rot -10.643 45.682 -2.695>
    scale {z=-0.105}
    translation=<-2.72 -2.025 -0.538>
   }
   n27 = "Transform (Modify) Bone - Bone: dyn_r_hair_a_02"(ComponentPose=@n26, Alpha=1, Translation="-0.184,-0.327,-1.905", Rotation="6.990,14.456,-2.060", Scale="1,1,1") {
    rotationMode=BMM_Additive
    rotationSpace=BCS_BoneSpace
    scaleMode=BMM_Additive
    scaleSpace=BCS_BoneSpace
    translationMode=BMM_Additive
    translationSpace=BCS_BoneSpace
    alphaScaleBiasClamp {bInterpResult=true, interpSpeedDecreasing=1, interpSpeedIncreasing=1}
    boneToModify {boneName=dyn_r_hair_a_02}
    rotation=<rot 6.99 14.456 -2.06>
    translation=<-0.184 -0.327 -1.905>
   }
   n28 = "Transform (Modify) Bone - Bone: dyn_l_hair_b_base"!(ComponentPose=@n27, bAlphaBoolEnabled:bool=True, Translation="0,0,0", Rotation="3.203,-6.360,7.062", Scale="1,1,1") -> (Pose) {
    alphaInputType=Bool
    rotationMode=BMM_Additive
    rotationSpace=BCS_BoneSpace
    translationMode=BMM_Additive
    translationSpace=BCS_BoneSpace
    alphaBoolBlend {blendInTime=0.3, blendOutTime=0.3}
    boneToModify {boneName=dyn_l_hair_b_base}
    rotation=<rot 3.203 -6.36 7.062>
   }
   n29 = "Transform (Modify) Bone - Bone: dyn_m_hair_base"!(ComponentPose=@n28, bAlphaBoolEnabled:bool=True, Translation="1.413,-0.003,-0.552", Rotation="5.391,10.138,-21.948", Scale="1,1,1") -> (Pose) {
    alphaInputType=Bool
    rotationMode=BMM_Additive
    rotationSpace=BCS_BoneSpace
    translationMode=BMM_Additive
    translationSpace=BCS_BoneSpace
    alphaBoolBlend {blendInTime=0.3, blendOutTime=0.3}
    boneToModify {boneName=dyn_m_hair_base}
    rotation=<rot 5.3909 10.138 -21.948>
    translation=<1.413 -0.003 -0.552>
   }
   n30 = "Transform (Modify) Bone - Bone: dyn_r_hair_b_base"!(ComponentPose=@n29, bAlphaBoolEnabled:bool=True, Translation="0,0,0", Rotation="14.384,14.534,-22.548", Scale="1,1,1") -> (Pose) {
    alphaInputType=Bool
    rotationMode=BMM_Additive
    rotationSpace=BCS_BoneSpace
    translationMode=BMM_Additive
    translationSpace=BCS_BoneSpace
    alphaBoolBlend {blendInTime=0.3, blendOutTime=0.3}
    boneToModify {boneName=dyn_r_hair_b_base}
    rotation=<rot 14.384 14.534 -22.548>
   }
   n31 = Component To Local(ComponentPose=@n30)
   n32 = Get bIsCrouching()
   n33 = RigidBody!(
    ComponentPose=Local To Component(LocalPose=Copy Pose From Mesh() {bUseAttachedParent=true})
    Alpha
    OverrideWorldGravity:Vector="0,0,0"
    ExternalForce
    SimSpaceSettings:Sim Space Settings Structure=(WorldAlpha=0,VelocityScaleZ=1,DampingAlpha=1,MaxLinearVelocity=10000,MaxAngularVelocity=10000,MaxLinearAcceleration=10000,MaxAngularAcceleration=10000,ExternalLinearDragV=(X=0,Y=0,Z=0),ExternalLinearVelocity=(X=0,Y=0,Z=0),ExternalAngularVelocity=(X=0,Y=0,Z=0))
   ) -> (Pose) {
    bOverrideWorldGravity=true
    overridePhysicsAsset=&/F_MED_FuzzyGlam_FaceAcc_Physics
    simulationSpace=BaseBoneSpace
    baseBoneRef {boneName=head}
   }
   n34 = Component To Local(ComponentPose=@n33)
   n35 = Anim Dynamics - Bone: dyn_l_feather_02!(
    ComponentPose=Local To Component(LocalPose=Blend Poses by bool(BlendPose_0=@n31, BlendPose_1=@n34, BlendTime_0=1, BlendTime_1=1, bActiveValue=@n32))
    Alpha=1
    GravityOverride:Vector="0,0,-980"
    ExternalForce="0,0,0"
   ) -> (Pose) {
    angularDampingOverride=1
    angularSpringConstant=12000
    bAngularSpring=true
    bOverrideAngularDamping=true
    bUseGravityOverride=true
    gravityScale=0.35
    simulationSpace=RootRelative
    boundBone {boneName=dyn_l_feather_02}
    componentAppliedLinearAccClamp=<500 500 500>
    componentLinearAccScale=<0.5 0.5 0.5>
    componentLinearVelScale=<0.5 0.5 0.5>
    gravityOverride {z=-980}
    physicsBodyDefinitions [{
      boundBone {boneName=dyn_l_feather_02}
      boxExtents=<20 30 30>
      constraintSetup {angularLimitsMax=<0 4 4>, angularLimitsMin=<0 -4 -4>, angularTarget=<1 0 0>}
      localJointOffset=<2 0 0>
     }]
   }
   n36 = Anim Dynamics - Bone: dyn_r_feather_02!(ComponentPose=@n35, Alpha=1, GravityOverride:Vector="0,0,-980", ExternalForce="0,0,0") -> (Pose) {
    angularDampingOverride=1
    angularSpringConstant=12000
    bAngularSpring=true
    bOverrideAngularDamping=true
    bUseGravityOverride=true
    gravityScale=0.35
    simulationSpace=RootRelative
    boundBone {boneName=dyn_r_feather_02}
    componentAppliedLinearAccClamp=<500 500 500>
    componentLinearAccScale=<0.5 0.5 0.5>
    componentLinearVelScale=<0.5 0.5 0.5>
    gravityOverride {z=-980}
    physicsBodyDefinitions [{
      boundBone {boneName=dyn_r_feather_02}
      boxExtents=<20 30 30>
      constraintSetup {angularLimitsMax=<0 4 4>, angularLimitsMin=<0 -4 -4>, angularTarget=<1 0 0>}
      localJointOffset=<2 0 0>
     }]
   }
   n37 = Anim Dynamics - Bone: dyn_f_hair!(ComponentPose=@n36, Alpha=1, GravityOverride:Vector="0,0,-980", ExternalForce="0,0,0") -> (Pose) {
    angularDampingOverride=1
    angularSpringConstant=20000
    bAngularSpring=true
    bOverrideAngularDamping=true
    bUseGravityOverride=true
    gravityScale=0.35
    simulationSpace=RootRelative
    boundBone {boneName=dyn_f_hair}
    componentAppliedLinearAccClamp=<5000 5000 5000>
    componentLinearAccScale=<0.5 0.5 0.5>
    componentLinearVelScale=<0.5 0.5 0.5>
    gravityOverride {z=-980}
    physicsBodyDefinitions [{
      boundBone {boneName=dyn_f_hair}
      boxExtents=<15 20 20>
      constraintSetup {angularLimitsMax=<0 0 5>, angularLimitsMin=<0 -15 -5>, angularTarget=<1 0 0>}
      localJointOffset=<1 0 0>
     }]
   }
   n38 = Anim Dynamics - Bone: dyn_r_hair_01!(ComponentPose=@n37, Alpha=1, GravityOverride:Vector="0,0,-980", ExternalForce="0,0,0") -> (Pose) {
    angularDampingOverride=1
    angularSpringConstant=3000
    bAngularSpring=true
    bOverrideAngularDamping=true
    bUseGravityOverride=true
    gravityScale=0.35
    simulationSpace=RootRelative
    boundBone {boneName=dyn_r_hair_01}
    componentAppliedLinearAccClamp=<5000 5000 5000>
    componentLinearAccScale=<0.2 0.2 0.2>
    componentLinearVelScale=<0.2 0.2 0.2>
    gravityOverride {z=-980}
    physicsBodyDefinitions [{
      boundBone {boneName=dyn_r_hair_01}
      boxExtents=<15 20 20>
      constraintSetup {angularLimitsMax=<0 0 2>, angularLimitsMin=<0 -2 -2>, angularTarget=<1 0 0>}
      localJointOffset=<1 0 0>
     }]
   }
   n39 = Anim Dynamics - Bone: dyn_l_hair_01!(ComponentPose=@n38, Alpha=1, GravityOverride:Vector="0,0,-980", ExternalForce="0,0,0") -> (Pose) {
    angularDampingOverride=1
    angularSpringConstant=3000
    bAngularSpring=true
    bOverrideAngularDamping=true
    bUseGravityOverride=true
    gravityScale=0.35
    simulationSpace=RootRelative
    boundBone {boneName=dyn_l_hair_01}
    componentAppliedLinearAccClamp=<5000 5000 5000>
    componentLinearAccScale=<0.2 0.2 0.2>
    componentLinearVelScale=<0.2 0.2 0.2>
    gravityOverride {z=-980}
    physicsBodyDefinitions [{
      boundBone {boneName=dyn_l_hair_01}
      boxExtents=<15 20 20>
      constraintSetup {angularLimitsMax=<0 0 2>, angularLimitsMin=<0 -2 -2>, angularTarget=<1 0 0>}
      localJointOffset=<1 0 0>
     }]
   }
   n40 = Anim Dynamics - Bone: dyn_r_ear_ring_01!(ComponentPose=@n39, Alpha=1, GravityOverride:Vector="0,0,-980", ExternalForce="0,0,0") -> (Pose) {
    angularDampingOverride=1
    angularSpringConstant=10000
    bAngularSpring=true
    bOverrideAngularDamping=true
    bUseGravityOverride=true
    gravityScale=0.35
    simulationSpace=RootRelative
    boundBone {boneName=dyn_r_ear_ring_01}
    componentAppliedLinearAccClamp=<3000 3000 3000>
    componentLinearAccScale=<0.5 0.5 0.5>
    componentLinearVelScale=<0.5 0.5 0.5>
    gravityOverride {z=-980}
    physicsBodyDefinitions [{
      boundBone {boneName=dyn_r_ear_ring_01}
      boxExtents=<10 20 20>
      constraintSetup {angularLimitsMax=<0 20 20>, angularLimitsMin=<0 -20 -20>, angularTarget=<1 0 0>}
      localJointOffset=<2 0 0>
     }]
   }
   n41 = Anim Dynamics - Bone: dyn_l_hair_a_01!(ComponentPose=@n40, Alpha=1, GravityOverride:Vector="0,0,-980", ExternalForce="0,0,0") -> (Pose) {
    angularDampingOverride=1
    angularSpringConstant=3000
    bAngularSpring=true
    bOverrideAngularDamping=true
    bUseGravityOverride=true
    gravityScale=0.35
    simulationSpace=RootRelative
    boundBone {boneName=dyn_l_hair_a_01}
    componentAppliedLinearAccClamp=<5000 5000 5000>
    componentLinearAccScale=<0.6 0.2 0.2>
    componentLinearVelScale=<0.2 0.2 0.2>
    gravityOverride {z=-980}
    physicsBodyDefinitions [{
      boundBone {boneName=dyn_l_hair_a_01}
      boxExtents=<20 30 30>
      constraintSetup {angularLimitsMax=<0 0 5>, angularLimitsMin=<0 -8 -20>, angularTarget=<1 0 0>}
      localJointOffset=<1 0 0>
     }]
   }
   n42 = Anim Dynamics - Bone: dyn_l_hair_a_02!(ComponentPose=@n41, Alpha=1, GravityOverride:Vector="0,0,-980", ExternalForce="0,0,0") -> (Pose) {
    angularDampingOverride=1
    angularSpringConstant=3000
    bAngularSpring=true
    bOverrideAngularDamping=true
    bUseGravityOverride=true
    gravityScale=0.35
    simulationSpace=RootRelative
    boundBone {boneName=dyn_l_hair_a_02}
    componentAppliedLinearAccClamp=<5000 5000 5000>
    componentLinearAccScale=<0.6 0.2 0.2>
    componentLinearVelScale=<0.2 0.2 0.2>
    gravityOverride {z=-980}
    physicsBodyDefinitions [{
      boundBone {boneName=dyn_l_hair_a_02}
      boxExtents=<20 30 30>
      constraintSetup {angularLimitsMax=<0 0 5>, angularLimitsMin=<0 -5 -5>, angularTarget=<1 0 0>}
      localJointOffset=<1 0 0>
     }]
   }
   n43 = Anim Dynamics - Bone: dyn_r_hair_a_01!(ComponentPose=@n42, Alpha=1, GravityOverride:Vector="0,0,-980", ExternalForce="0,0,0") -> (Pose) {
    angularDampingOverride=1
    angularSpringConstant=3000
    bAngularSpring=true
    bOverrideAngularDamping=true
    bUseGravityOverride=true
    gravityScale=0.35
    simulationSpace=RootRelative
    boundBone {boneName=dyn_r_hair_a_01}
    componentAppliedLinearAccClamp=<5000 5000 5000>
    componentLinearAccScale=<0.5 0.2 0.2>
    componentLinearVelScale=<0.2 0.2 0.2>
    gravityOverride {z=-980}
    physicsBodyDefinitions [{
      boundBone {boneName=dyn_r_hair_a_01}
      boxExtents=<15 20 20>
      constraintSetup {angularLimitsMax=<0 0 8>, angularLimitsMin=<0 -8 -8>, angularTarget=<1 0 0>}
      localJointOffset=<1 0 0>
     }]
   }
   n44 = Anim Dynamics - Bone: dyn_r_hair_a_02!(ComponentPose=@n43, Alpha=1, GravityOverride:Vector="0,0,-980", ExternalForce="0,0,0") -> (Pose) {
    angularDampingOverride=1
    angularSpringConstant=3000
    bAngularSpring=true
    bOverrideAngularDamping=true
    bUseGravityOverride=true
    gravityScale=0.35
    simulationSpace=RootRelative
    boundBone {boneName=dyn_r_hair_a_02}
    componentAppliedLinearAccClamp=<5000 5000 5000>
    componentLinearAccScale=<0.5 0.2 0.2>
    componentLinearVelScale=<0.2 0.2 0.2>
    gravityOverride {z=-980}
    physicsBodyDefinitions [{
      boundBone {boneName=dyn_r_hair_a_02}
      boxExtents=<15 20 20>
      constraintSetup {angularLimitsMax=<0 0 5>, angularLimitsMin=<0 -5 -5>, angularTarget=<1 0 0>}
      localJointOffset=<1 0 0>
     }]
   }
   n45 = "Transform (Modify) Bone - Bone: dyn_l_hair_b_base"!(ComponentPose=@n44, Alpha:float=1, Translation="2.323,-0.692,-0.712", Rotation="-34.926,-27.279,18.745", Scale="1,1,1") -> (Pose) {
    rotationMode=BMM_Additive
    rotationSpace=BCS_BoneSpace
    translationMode=BMM_Additive
    translationSpace=BCS_BoneSpace
    alphaScaleBiasClamp {bInterpResult=true}
    boneToModify {boneName=dyn_l_hair_b_base}
    rotation=<rot -34.926 -27.279 18.745>
    translation=<2.323 -0.692 -0.712>
   }
   n46 = "Transform (Modify) Bone - Bone: dyn_m_hair_base"!(ComponentPose=@n45, Alpha:float=1, Translation="0,0,0", Rotation="9.980,-17.323,-1.779", Scale="1,1,1") -> (Pose) {
    rotationMode=BMM_Additive
    rotationSpace=BCS_BoneSpace
    scaleMode=BMM_Additive
    scaleSpace=BCS_BoneSpace
    translationMode=BMM_Additive
    translationSpace=BCS_BoneSpace
    alphaScaleBiasClamp {bInterpResult=true}
    boneToModify {boneName=dyn_m_hair_base}
    rotation=<rot 9.98 -17.323 -1.779>
   }
   n47 = "Transform (Modify) Bone - Bone: dyn_r_hair_b_base"!(ComponentPose=@n46, Alpha:float=1, Translation="1.247,-0.633,-0.765", Rotation="45.978,158.030,-145.674", Scale="1.000,0.813,1.000") -> (Pose) {
    rotationMode=BMM_Additive
    rotationSpace=BCS_BoneSpace
    scaleMode=BMM_Additive
    scaleSpace=BCS_BoneSpace
    translationMode=BMM_Additive
    translationSpace=BCS_BoneSpace
    alphaScaleBiasClamp {bInterpResult=true}
    boneToModify {boneName=dyn_r_hair_b_base}
    rotation=<rot 45.978 158.03 -145.674>
    scale {y=0.8132}
    translation=<1.247 -0.633 -0.765>
   }
   n48 = "Transform (Modify) Bone - Bone: dyn_r_hair_b_02"(ComponentPose=@n47, Alpha=1, Translation="-1.387,2.696,-2.260", Rotation="-20.233,15.855,-23.893", Scale="0.446,1.000,1.000") {
    rotationMode=BMM_Additive
    rotationSpace=BCS_BoneSpace
    scaleMode=BMM_Additive
    scaleSpace=BCS_BoneSpace
    translationMode=BMM_Additive
    translationSpace=BCS_BoneSpace
    alphaScaleBiasClamp {bInterpResult=true}
    boneToModify {boneName=dyn_r_hair_b_02}
    rotation=<rot -20.233 15.855 -23.893>
    scale {x=0.446}
    translation=<-1.387 2.696 -2.26>
   }
   n49 = "Transform (Modify) Bone - Bone: dyn_r_hair_a_01"(ComponentPose=@n48, Alpha=1, Translation="-2.884,-1.124,-0.041", Rotation="2.230,48.888,1.241", Scale="1,1,1") {
    rotationMode=BMM_Additive
    rotationSpace=BCS_BoneSpace
    scaleMode=BMM_Additive
    scaleSpace=BCS_BoneSpace
    translationMode=BMM_Additive
    translationSpace=BCS_BoneSpace
    alphaScaleBiasClamp {bInterpResult=true}
    boneToModify {boneName=dyn_r_hair_a_01}
    rotation=<rot 2.23 48.888 1.241>
    translation=<-2.884 -1.124 -0.041>
   }
   n50 = "Transform (Modify) Bone - Bone: dyn_l_hair_a_base"(ComponentPose=@n49, Alpha=1, Translation="0,0,0", Rotation="0.006,-70.000,-0.003", Scale="1,1,1") {
    rotationMode=BMM_Additive
    rotationSpace=BCS_BoneSpace
    scaleMode=BMM_Additive
    scaleSpace=BCS_BoneSpace
    translationMode=BMM_Additive
    translationSpace=BCS_BoneSpace
    alphaScaleBiasClamp {bInterpResult=true}
    boneToModify {boneName=dyn_l_hair_a_base}
    rotation=<rot 0.006 -69.9999 -0.003>
   }
   n51 = "Transform (Modify) Bone - Bone: dyn_l_hair_a_02"(ComponentPose=@n50, Alpha=1, Translation="0.271,2.332,0.016", Rotation="-0.020,90.000,-0.015", Scale="1,1,1") {
    rotationMode=BMM_Additive
    rotationSpace=BCS_BoneSpace
    scaleMode=BMM_Additive
    scaleSpace=BCS_BoneSpace
    translationMode=BMM_Additive
    translationSpace=BCS_BoneSpace
    alphaScaleBiasClamp {bInterpResult=true}
    boneToModify {boneName=dyn_l_hair_a_02}
    rotation=<rot -0.02 90 -0.015>
    translation=<0.271 2.332 0.016>
   }
   n52 = "Transform (Modify) Bone - Bone: dyn_r_hair_a_base1"(ComponentPose=@n51, Alpha=1, Translation="0,0,0", Rotation="0,35,0", Scale="1,1,1") {
    rotationMode=BMM_Additive
    rotationSpace=BCS_BoneSpace
    scaleMode=BMM_Additive
    scaleSpace=BCS_BoneSpace
    translationMode=BMM_Additive
    translationSpace=BCS_BoneSpace
    alphaScaleBiasClamp {bInterpResult=true}
    boneToModify {boneName=dyn_r_hair_a_base1}
    rotation {yaw=35}
   }
   n53 = Get IsWearingBackbling() -> (IsWearingBackbling:bool)
   n54 = "Constraint - dyn_l_hair_b_01 (PPP)"!(ComponentPose=@n52, bAlphaBoolEnabled:bool=@n53, ConstraintWeights_0=0.2, ConstraintWeights_1:float=0.8, ConstraintWeights_2:float=0) -> (Pose) {
    alphaInputType=Bool
    alphaBoolBlend {blendOption=Linear}
    boneToModify {boneName=dyn_l_hair_b_01}
    constraintSetup [{targetBone {boneName=spine_05}}, {targetBone {boneName=neck_01}}, {targetBone {boneName=clavicle_l}}]
    constraintWeights [1,1,1]
   }
   n55 = "Constraint - dyn_m_hair_01 (PPP)"!(ComponentPose=@n54, bAlphaBoolEnabled:bool=@n53, ConstraintWeights_0=0.2, ConstraintWeights_1:float=0.8, ConstraintWeights_2:float=0) -> (Pose) {
    alphaInputType=Bool
    alphaBoolBlend {blendOption=Linear}
    boneToModify {boneName=dyn_m_hair_01}
    constraintSetup [{targetBone {boneName=spine_05}}, {targetBone {boneName=neck_01}}, {targetBone {boneName=clavicle_l}}]
    constraintWeights [1,1,1]
   }
   n56 = "Constraint - dyn_r_hair_b_01 (PPP)"!(ComponentPose=@n55, bAlphaBoolEnabled:bool=@n53, ConstraintWeights_0=0.2, ConstraintWeights_1:float=0.8, ConstraintWeights_2:float=0) -> (Pose) {
    alphaInputType=Bool
    alphaBoolBlend {blendOption=Linear}
    boneToModify {boneName=dyn_r_hair_b_01}
    constraintSetup [{targetBone {boneName=spine_05}}, {targetBone {boneName=neck_01}}, {targetBone {boneName=clavicle_r}}]
    constraintWeights [1,1,1]
   }
   n57 = "Constraint - dyn_l_hair_b_02 (PPP)"!(ComponentPose=@n56, bAlphaBoolEnabled:bool=@n53, ConstraintWeights_0=0.05, ConstraintWeights_1:float=0.9, ConstraintWeights_2:float=0.05) -> (Pose) {
    alphaInputType=Bool
    alphaBoolBlend {blendOption=Linear}
    boneToModify {boneName=dyn_l_hair_b_02}
    constraintSetup [{targetBone {boneName=spine_05}}, {targetBone {boneName=neck_01}}, {targetBone {boneName=clavicle_l}}]
    constraintWeights [1,1,1]
   }
   n58 = "Constraint - dyn_m_hair_02 (PPP)"!(ComponentPose=@n57, bAlphaBoolEnabled:bool=@n53, ConstraintWeights_0=0.1, ConstraintWeights_1:float=0.9, ConstraintWeights_2:float=0) -> (Pose) {
    alphaInputType=Bool
    alphaBoolBlend {blendOption=Linear}
    boneToModify {boneName=dyn_m_hair_02}
    constraintSetup [{targetBone {boneName=spine_05}}, {targetBone {boneName=neck_01}}, {targetBone {boneName=clavicle_l}}]
    constraintWeights [1,1,1]
   }
   n59 = "Constraint - dyn_r_hair_b_02 (PPP)"!(ComponentPose=@n58, bAlphaBoolEnabled:bool=@n53, ConstraintWeights_0=0.1, ConstraintWeights_1:float=0.85, ConstraintWeights_2:float=0.05) -> (Pose) {
    alphaInputType=Bool
    alphaBoolBlend {blendOption=Linear}
    boneToModify {boneName=dyn_r_hair_b_02}
    constraintSetup [{targetBone {boneName=spine_05}}, {targetBone {boneName=neck_01}}, {targetBone {boneName=clavicle_r}}]
    constraintWeights [1,1,1]
   }
   Output Pose(Result=Component To Local(ComponentPose=@n59)) {layerGroup=None}
  ]}
 EventGraph {nodes [
   Event BlueprintUpdateAnimation()
   > n1 = Sequence!(execute) -> (then_0, then_1, then_2:Exec, then_3:Exec, then_4:Exec, then_5:Exec, then_6:Exec)
   > n2 = "Gravity Override State Function (Parameter Asset) (Nativized)"(
    CharacterPartAnimInstance=Self-Reference()
    GameTimeInSeconds=GetGameTimeInSeconds()
    Parameters=&/F_MED_FuzzyGlam_FaceAcc_Parameters
   )
   > Set GravityOverride(GravityOverride=@n2)
   > Set RBAN Settings Out(RBAN Settings Out=@n2.OutRigidBodyAnimNodeParameters)
   Set CrouchWalk(execute=@n1.then_1, CrouchWalk:bool=AND Boolean(A=Get bIsCrouching(), B=Get bIsMoving2D())) -> (Output_Get:bool)
   Set CrouchTargeting(execute=@n1.then_2, CrouchTargeting:bool=AND Boolean(A=Get bIsCrouching(), B=Get bIsTargeting())) -> (Output_Get:bool)
   n3 = Set Head Up(
    execute=@n1.then_3
    Head Up:double=GetAlphaFromReferenceTransformDelta(SkeletalMeshComponent=Get Skel Mesh(), BoneName=spine_05, BaseName=head, InputRange=(X=0,Y=90), TwistAxis="0,0,1")
   ) -> (Output_Get:double)
   > Set CrouchingHeadUP(CrouchingHeadUP:double=SelectFloat(A=@n3, B=0, bPickA=Get CrouchTargeting() -> (CrouchTargeting:bool))) -> (Output_Get:double)
   n4 = NOT Boolean(A=Get bIsCrouching())
   n5 = Get bIsTargeting()
   > Set AimUp(AimUp:double=SelectFloat(A=@n3, B=0, bPickA=AND Boolean(A=@n4, B=@n5))) -> (Output_Get:double)
   n6 = NOT Boolean(A=Get bIsCrouching())
   n7 = Get bIsTargeting()
   > Set Head Up_Target(Head Up_Target:double=SelectFloat(A=Get Head Up() -> (Head Up:double), B=0, bPickA=AND Boolean(A=@n7, B=@n6))) -> (Output_Get:double)
   n8 = CheckForBackbling()
   Set IsWearingBackbling(execute=@n1.then_5, IsWearingBackbling:bool=OR Boolean!(A=@n8, B=@n8.bIsBackblingACape, C:bool=@n8.bIsBackblingLong) -> (ReturnValue)) -> (Output_Get:bool)
   > Set Is Backbling A Cape(Is Backbling A Cape:bool=@n8.bIsBackblingACape) -> (Output_Get:bool)
   > Set is Wear a Cape(is Wear a Cape:bool=AND Boolean(A=@n8, B=@n8.bIsBackblingLong)) -> (Output_Get:bool)
   Event BlueprintInitializeAnimation()
   > Set Skel Mesh(Skel Mesh=GetOwnerSkeletalMeshComponent())
   TryGetPawnOwner()
  ]}
]
variables [
 AimUp {type=number}
 AlphaJumping {type=number}
 CrouchAlpha {type=number}
 CrouchingHeadUP {type=number}
 CrouchingHeadUP_Bool {type=boolean}
 CrouchTargeting {type=boolean}
 CrouchWalk {type=boolean}
 GravityOverride {type=Vector}
 Head Down {type=number}
 Head Up {type=number}
 Head Up_Modified {type=number}
 Head Up_Target {type=number}
 Head_Tilt_R {type=number}
 Head_Tilt_R_Bool {type=boolean}
 Head_Turn_L {type=number}
 Head_Turn_L_B {type=boolean}
 Head_Turn_R {type=number}
 HeadUP Left {type=number}
 HeadUP Left_Float {type=number}
 HeadUP LeftTilt {type=boolean}
 Is Backbling A Cape {type=boolean}
 is Wear a Cape {type=boolean}
 IsWearingBackbling {type=boolean}
 LookUpCrouchAlpha {type=number}
 LookUpPoseAlpha {type=number}
 RBAN Settings Out {type=FortRigidBodyAnimNodeParameters}
 Skel Mesh {type=SkeletalMeshComponent}
]
